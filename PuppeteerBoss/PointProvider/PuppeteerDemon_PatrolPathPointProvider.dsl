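; Spawn point provider for the Puppeteer Demon boss that passes level references to the boss once it spawns.
; On the server, it listens for a successful spawn, stores the actor, casts it to the boss pawn classes, and sets its arena box and leash volume.
; It then notifies the ghost manager and binds to the boss's mask-spawn event to give each valid mask its teleport area box.
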(event EventBeginPlay
  (Networking|SwitchHasAuthority
    (:Authority
      (EventDispatchers|BindEventtoOnActorSpawnedResult (delegate Custom|ActorSpawned)))))

(event Custom|ActorSpawned
  (params (SpawnedActor :type Actor%20Object%20Reference) (bResult :type Boolean))
  (if bResult
    (bind _output_get (Variables|Default|SetSpawnedActor SpawnedActor))
    (bind _asbp_player_pawn_fire_petal_boss (Utilities|Casting|CastToBP_PlayerPawn_FirePetal_Boss _output_get))
    (Variables|/FirePetalBosses/Pawn/BP_PlayerPawn_FirePetal_Boss.BP_PlayerPawn_FirePetal_Boss_C|SetBossArenaBox (Variables|Default|GetPuppeteerBossArenaBox) _asbp_player_pawn_fire_petal_boss)
    (bind _asbp_player_pawn_puppeteer_demon (Utilities|Casting|CastToBP_PlayerPawn_Puppeteer_Demon _asbp_player_pawn_fire_petal_boss))
    (Variables|/PuppeteerBoss/Pawn/BP_PlayerPawn_Puppeteer_Demon.BP_PlayerPawn_Puppeteer_Demon_C|SetLeashVolume (Variables|Default|GetPuppeteerLeashVolume) _asbp_player_pawn_puppeteer_demon)
    (Class|GhostManagerBP|PuppeteerDemonSpawned (Variables|Default|GetGhostManager) _asbp_player_pawn_puppeteer_demon)
    (Default|BindEventtoMaskActorSpawned _asbp_player_pawn_puppeteer_demon (delegate Custom|OnMaskActorSpawned))))

(event Custom|OnMaskActorSpawned
  (params (MaskActor :type BGA%20Puppeteer%20Mask%20Object%20Reference))
  (Utilities|IsValid MaskActor
    (:Is%20Valid
      (Variables|/PuppeteerBoss/Mask/BGA_PuppeteerMask.BGA_PuppeteerMask_C|SetTeleportAreaBox (Variables|Default|GetMaskTeleportAreaBox) MaskActor)
      (Variables|/PuppeteerBoss/Mask/BGA_PuppeteerMask.BGA_PuppeteerMask_C|SetTeleportAreaBoxValidated true MaskActor))))

(fn ConstructionScript)
