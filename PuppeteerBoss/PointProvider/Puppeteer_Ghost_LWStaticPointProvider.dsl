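; Static point provider for Puppeteer boss ghosts that initializes a ghost pawn once it spawns.
; On authority, it binds to its OnActorSpawned event and casts the spawned actor to BP_PlayerPawn_Puppeteer_Ghost.
; With a valid suitable player, it passes the ghost its spawn area and target player, then binds it to the demon pawn or destroys it if none exists.
; The provider also destroys itself during this handoff, making it single-use.
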(event EventBeginPlay
  (Networking|SwitchHasAuthority
    (:Authority
      (EventDispatchers|BindEventtoOnActorSpawned (delegate Custom|ActorSpawned)))))

(event Custom|ActorSpawned
  (params (SpawnedActor :type Actor%20Object%20Reference))
  (bind _demonpawn (Variables|Default|GetDemonPawn))
  (bind _ghost_spawn_area (Variables|Default|GetGhost%20Spawn%20Area))
  (bind _suitableplayer (Variables|Default|GetSuitablePlayer))
  (bind _output_get (Variables|Default|SetSpawnedActor SpawnedActor))
  (bind _asbp_player_pawn_puppeteer_ghost (Utilities|Casting|CastToBP_PlayerPawn_Puppeteer_Ghost _output_get)
    (:then
      (Utilities|IsValid _suitableplayer
        (:Is%20Valid
          (Utilities|IsValid _ghost_spawn_area
            (:Is%20Valid
              (Variables|/PuppeteerBoss/Pawn/BP_PlayerPawn_Puppeteer_Ghost.BP_PlayerPawn_Puppeteer_Ghost_C|SetGhost%20Spawn%20Area _ghost_spawn_area _asbp_player_pawn_puppeteer_ghost))
            (:Is%20Not%20Valid
              _))
          (Variables|/PuppeteerBoss/Pawn/BP_PlayerPawn_Puppeteer_Ghost.BP_PlayerPawn_Puppeteer_Ghost_C|SetTargetPlayer _suitableplayer _asbp_player_pawn_puppeteer_ghost)
          (Utilities|IsValid _demonpawn
            (:Is%20Valid
              (Class|BPPlayerPawnPuppeteerGhost|BindtoDemonPawn _asbp_player_pawn_puppeteer_ghost _demonpawn))
            (:Is%20Not%20Valid
              (Actor|DestroyActor _asbp_player_pawn_puppeteer_ghost)))
          (Actor|DestroyActor))
        (:Is%20Not%20Valid
          _)))
    (:CastFailed
      _)))

(fn ConstructionScript)
