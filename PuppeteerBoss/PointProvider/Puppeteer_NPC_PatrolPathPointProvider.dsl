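; Patrol path point provider for the Puppeteer boss NPC that delays when the NPC's patrol becomes available for Living World spawning.
; On authority, it waits for match ready and initial calendar sync, then checks its calendar event.
; If the event is active, it enables itself after a scalable delay.
; Otherwise it binds to the demon patrol provider's spawn result and enables itself after that delay once the Puppeteer demon dies.
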(event EventBeginPlay
  (Networking|SwitchHasAuthority
    (:Authority
      (Class|FortAsyncActionAuthorityMatchReady|MatchReadyAsync
        (:OnReady
          (MCPContext|InitialCalendarSyncCompleted
            (:OnSyncComplete
              (bind _biseventactive (CalendarEvents|GetCalendarEventInformation (Variables|Default|GetCalendarEvent)))
              (if _biseventactive
                (Utilities|FlowControl|Delay (ScalableFloat|GetValueatLevel (Variables|Default|GetPuppeteerNPCSpawnDelay_Hf)))
                (Class|FortAthenaLivingWorldPointProviderInterface|EnablePointProvider_DEPRECATED)
                (else
                  (EventDispatchers|BindEventtoOnActorSpawnedResult (Variables|Default|GetDemonPatrolPointProvider) (delegate Custom|DemonSpawned)))))))))))

(event Custom|DemonSpawned
  (params (SpawnedActor :type Actor%20Object%20Reference) (bResult :type Boolean))
  (bind _asbp_player_pawn_puppeteer_demon (Utilities|Casting|CastToBP_PlayerPawn_Puppeteer_Demon SpawnedActor))
  (EventDispatchers|BindEventtoOnDied _asbp_player_pawn_puppeteer_demon (delegate Custom|PuppeteerDied)))

(event Custom|PuppeteerDied
  (params (DamagedActor :type Actor%20Object%20Reference) (Damage :type Float%20%28single-precision%29) (InstigatedBy :type Controller%20Object%20Reference) (DamageCauser :type Actor%20Object%20Reference) (HitLocation :type Vector) (FHitComponent :type Primitive%20Component%20Object%20Reference) (BoneName :type Name) (Momentum :type Vector))
  (Utilities|FlowControl|Delay (ScalableFloat|GetValueatLevel (Variables|Default|GetPuppeteerNPCSpawnDelay_Hf)))
  (Class|FortAthenaLivingWorldPointProviderInterface|EnablePointProvider_DEPRECATED))

(fn ConstructionScript)
